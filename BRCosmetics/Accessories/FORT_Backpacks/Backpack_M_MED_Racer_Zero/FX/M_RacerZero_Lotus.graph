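# Lit material for the Racer Zero backpack's lotus, with petals that animate open and closed and glow from the center.
# An Open blend, timed from OpenTime, plus a sine breathe drives vertex rotation about UV1-derived axes, with offset and corrected normals.
# Emissive combines a center sphere mask, cracked-glass normals and masks, lotus texture and Fresnel.
# Glow follows openness and time of day, which also tints base color brightness.
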Material /BRCosmetics/Accessories/FORT_Backpacks/Backpack_M_MED_Racer_Zero/FX/M_RacerZero_Lotus
graph {
 properties {
  bEnableResponsiveAA=true
  bUsedWithMeshDeformer=true
  bUsedWithSkeletalMesh=true
  pixelDepthOffsetMode=PDOM_Legacy
  usedShadingModels=Default Lit
 }
 nodes [
  n1 = Constant()
  n2 = MaterialFunctionCall(Float2=Subtract(A=TextureCoordinate() {coordinateIndex=1}) {constB=0.5}) {materialFunction=&BreakOutFloat2Components}
  n3 = Normalize(VectorInput=Transform(None=MaterialFunctionCall(X=@n1, Y=@n2.G, Z=@n2.G) {materialFunction=&MakeFloat3}.Result) {transformSourceType=TRANSFORMSOURCE_Local})
  n4 = Time()
  n5 = ScalarParameter() {parameterName=OpenTime}
  n6 = Saturate(None=Divide(A=Subtract(A=@n4, B=@n5), B=ScalarParameter() {defaultValue=4, parameterName=OpenDuration}))
  n7 = LinearInterpolate(A=OneMinus(None=@n6), B=@n6, Alpha=ScalarParameter() {parameterName=Open})
  n8 = ScalarParameter() {defaultValue=0.22, parameterName=BreatheSpeed}
  n9 = Add(
   A=@n7
   B=Multiply(
    A=MaterialFunctionCall(Speed=@n8) {materialFunction=&DebugTimeSine}.Result
    B=ScalarParameter() {defaultValue=0.1, parameterName=BreatheAmoumt}
   )
  )
  n10 = SphereMask(
   A=WorldPosition().XYZ
   B=MaterialFunctionCall() {materialFunction=&ObjectPivotPoint}.Object Pivot Location
   Radius=ScalarParameter() {defaultValue=35, parameterName=WPORadius}
  ) {hardnessPercent=0}
  n11 = Multiply(
   A=LinearInterpolate(
    A=ScalarParameter() {defaultValue=-0.068, parameterName=ClosedRotation}
    B=ScalarParameter() {defaultValue=0.068, parameterName=OpenRotation}
    Alpha=@n9
   )
   B=LinearInterpolate(Alpha=Power(Base=@n10) {constExponent=3}) {constA=1, constB=0}
  )
  n12 = VectorParameter() {parameterName=OpenOffset, defaultValue=<rgba 1.7 0 0 1>}
  Add(
   A=RotateAboutAxis(
    NormalizedRotationAxis=@n3
    RotationAngle=@n11
    PivotPoint=MaterialFunctionCall() {materialFunction=&ObjectPivotPoint}.Object Pivot Location
    Position=WorldPosition().XYZ
   )
   B=Transform(None=Multiply(A=@n7, B=@n12.RGB)) {transformSourceType=TRANSFORMSOURCE_Local}
  )
  Constant() {r=0.07}
  Constant() {r=0.7}
  n13 = Add(A=TextureCoordinate(), B=Multiply(A=TextureCoordinate() {coordinateIndex=2}) {constB=3.5})
  n14 = Transform(None=TextureSample(UVs=@n13) {samplerType=SAMPLERTYPE_Normal, texture=/Game/Effects/Fort_Effects/Textures/Tile/T_Glass_Cracks_N}.RGB)
  n15 = TransformPosition(None=VectorParameter() {parameterName=CenterPoint, defaultValue=<rgba 6 0 0 1>}.RGB) {transformType=TRANSFORMPOSSOURCE_World}
  n16 = WorldPosition()
  n17 = Normalize(VectorInput=Subtract(A=@n15, B=@n16.XYZ))
  n18 = SphereMask(A=@n16.XYZ, B=@n15, Radius=Add(A=ScalarParameter() {defaultValue=17, parameterName=Radius}, B=Multiply(A=@n9) {constB=4})) {hardnessPercent=0}
  n19 = Add(A=Multiply(A=DotProduct(A=@n14, B=@n17)) {constB=0.25}, B=Power(Base=@n18) {constExponent=2.5})
  n20 = VectorParameter() {parameterName=PetalColor, defaultValue=<rgba 0.0718 0.3179 1 0>}
  n21 = VectorParameter() {parameterName=EdgeColor, defaultValue=<rgba 0.28 0.8291 1 0>}
  n22 = Fresnel(ExponentIn=ScalarParameter() {defaultValue=2, parameterName=FresnelExp}) {baseReflectFraction=0, exponent=3}
  n23 = Add(
   A=Multiply(A=TextureSample() {texture=&/T_RacerZero_Lotus}.G) {constB=1.25}
   B=Multiply(A=TextureSample(UVs=@n13) {texture=/Game/Effects/Fort_Effects/Textures/Tile/T_Glass_Cracks_M}.G) {constB=0.75}
  )
  n24 = Add(A=Multiply(A=Add(A=@n18, B=@n22)) {constB=0.4}, B=@n23)
  n25 = Saturate(None=@n24)
  n26 = Add(
   A=Multiply(A=VectorParameter() {parameterName=CenterGlow, defaultValue=<rgba 1.3286 0.2583 2 0>}.RGB, B=Power(Base=@n19) {constExponent=3})
   B=Multiply(A=LinearInterpolate(A=@n20.RGB, B=@n21.RGB, Alpha=@n25), B=Add(A=@n24) {constB=0.1})
  )
  n27 = Constant() {r=0.7}
  n28 = MaterialFunctionCall(
   Color_Morning=@n27
   Color_Day=Constant() {r=1}
   Color_Evening=@n27
   Color_Night=VectorParameter() {parameterName=NightColor, defaultValue=<rgba 0.3453 0.1768 0.35 0>}.RGB
  ) {materialFunction=&Color_Time_of_Day}
  MaterialFunctionCall(Emissive=Multiply(A=Multiply(A=@n26, B=@n28.Result), B=LinearInterpolate(Alpha=@n9) {constA=0.4})) {materialFunction=&MF_FX_EmissiveAdjustments}
  MaterialFunctionCall(Rotation Axis=@n3, Rotation Angle=@n11, World Space Vertex Normals=@n14) {materialFunction=&FixRotateAboutAxisNormals}
  n29 = CollectionParameter() {
   collection=&FortniteMaterialParameters
   parameterName=FortniteActiveTimeOfDay
  }
  Multiply(
   A=LinearInterpolate(Alpha=ComponentMask(None=@n29) {a=true}) {constA=1, constB=0.6}
   B=Multiply(A=VectorParameter() {parameterName=BaseColor, defaultValue=<rgba 0.1704 0.4701 0.5 0>}.RGB, B=Add(A=@n23) {constB=0.25})
  )
 ]
}
